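# Conversation database for the SpringBreak NPC, Captain Jack Sparrow, defining the root dialogue that opens when a player talks to him.
# A single entry point leads to a greeting Speech node, whose default line is overridden by prioritized contextual lines for a first-ever meeting or three specific player outfits.
# The greeting branches through link nodes into separately tagged conversations: NPC quests, basic services, SpringBreak dialogue and two Jack-specific S30 entries.
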ConversationDatabase /NPCLibrary/NPCs/Greenhouse_NPCs/SpringBreak/NPCLibrary_Conversation_SpringBreak
entry_points [{tag=Conversation.Entry.SpringBreak, destinations [#1]}]
nodes [
 ConversationEntryPointNode_7 {
  id=#1
  class=ConversationEntryPointNode
  outputs [#2]
  properties {entryTag {tagName=Conversation.Entry.SpringBreak}}
 }
 Speech_C_4 {
  id=#2
  class=Speech_C
  outputs [
   #3
   #4
   #5
   #6
   #7
  ]
  properties {
   nodeName=""
   generalConfig {
    defaultMessage=Is this a dream? No? Thought not.
    contextualMessages [
     {
      message=I'm Captain Jack Sparrow. Savvy?
      contextRequirements [{requirement {class=FortControllerRequirement_IsFirstEverConversationWithNPC}}]
      priority {curve {rowName=Tandem.Speech.Priority.FirstEverInteraction}}
     }
     {
      message="The world's still the same, Hector. There's just...less in it."
      contextRequirements [{requirement {class=FortControllerRequirement_HasCID, properties {allowedCIds [/BRCosmetics/Athena/Items/Cosmetics/Characters/Character_DomeRoof]}}}]
      priority {curve {rowName=Tandem.Speech.Priority.OutfitReaction}}
     }
     {
      message="My debt was paid, mate."
      contextRequirements [{requirement {class=FortControllerRequirement_HasCID, properties {allowedCIds [/BRCosmetics/Athena/Items/Cosmetics/Characters/Character_AccentWall]}}}]
      priority {curve {rowName=Tandem.Speech.Priority.OutfitReaction}}
     }
     {
      message="I saved your life, you saved mine. We're square."
      contextRequirements [{requirement {class=FortControllerRequirement_HasCID, properties {allowedCIds [/BRCosmetics/Athena/Items/Cosmetics/Characters/Character_StoneLion]}}}]
      priority {curve {rowName=Tandem.Speech.Priority.OutfitReaction}}
     }
    ]
   }
  }
 }
 ConversationLinkNode_4 {id=#7, class=ConversationLinkNode, properties {remoteEntryTag {tagName=Conversation.Entry.S30.D01.Q05.Jack}}}
 ConversationLinkNode_3 {id=#6, class=ConversationLinkNode, properties {remoteEntryTag {tagName=Conversation.Entry.S30.D3.03.Jack}}}
 ConversationLinkNode_2 {
  id=#5
  class=ConversationLinkNode
  properties {remoteEntryTag {tagName=Conversation.Entry.SpringBreak.Dialogue}}
 }
 ConversationLinkNode_0 {id=#4, class=ConversationLinkNode, properties {remoteEntryTag {tagName=Conversation.Entry.BasicServices}}}
 ConversationLinkNode_1 {id=#3, class=ConversationLinkNode, properties {remoteEntryTag {tagName=Conversation.Entry.NPCLibrary.Quests}}}
]
